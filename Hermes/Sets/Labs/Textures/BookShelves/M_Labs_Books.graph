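# Masked metallic material for holographic books on instanced Labs bookshelves, with scanlines, fresnel dimming and a gradient glow.
# Per-instance custom data sets atlas flipbook columns and, with per-instance random, picks each book's frame.
# Distance-faded scanlines modulate exposure-compensated emissive; the Mask Maker Take Over switch swaps in distorted UVs and alternate tinting.
# Platform switches vary roughness and dithered opacity, and a shared destructible damage function wraps the result.
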Material /Hermes/Sets/Labs/Textures/BookShelves/M_Labs_Books
graph {
 properties {
  bAutomaticallySetUsageInEditor=false
  bUseMaterialAttributes=true
  bUsedWithInstancedStaticMeshes=true
  blendMode=BLEND_Masked
  usedShadingModels=Default Lit
 }
 nodes [
  n1 = PerInstanceCustomData!(DefaultValue) {constDefaultValue=4}
  n2 = Constant() {r=16}
  n3 = PerInstanceRandom()
  n4 = MaterialFunctionCall(Input=@n3, Target Low=Constant() {r=0.89}, Target High=Constant() {r=0.99}) {materialFunction=&RemapValueRange}
  n5 = MaterialFunctionCall(Input=@n3, Target Low=Constant() {r=0.5}, Target High=Constant() {r=0.87}) {materialFunction=&RemapValueRange}
  n6 = If(A=@n1, B=Constant() {r=8}, "A > B"=@n5.Result, "A == B"=@n5.Result, "A < B"=Multiply(A=@n3) {constB=0.499})
  n7 = NamedRerouteDeclaration(None=MaterialFunctionCall(
    "Animation  Phase (0-1)"=If(A=@n1, B=@n2, "A > B"=@n4.Result, "A == B"=@n4.Result, "A < B"=@n6)
    Number of Rows=Constant() {r=4}
    Number of Columns=@n1
   ) {materialFunction=&FlipBook}.UVs) {name=UVs, nodeColor=<rgba 1 0 0.4687 1>}
  n8 = TextureSample(UVs=StaticSwitchParameter(True=NamedRerouteUsage(), False=@n7) {group=Projector, parameterName=Mask Maker Take Over, sortPriority=1}) {samplerSource=SSM_Wrap_WorldGroupSettings, texture=&/T_Labs_BookAtlas}
  n9 = Max(A=Max(A=@n8.R, B=@n8.R), B=@n8.B)
  n10 = ScalarParameter() {defaultValue=1, parameterName=Scanline Speed 1}
  n11 = Multiply(A=ConstantBiasScale(None=@n3) {bias=-0.5, scale=2}) {constB=0.03}
  n12 = ConstantBiasScale(None=MaterialFunctionCall!(
    ScanLine_Tiling_1=ScalarParameter() {defaultValue=10000, parameterName=Scanline Tiling 1}
    Scanline_PannerSpeed_1=Add(A=@n10, B=@n11)
    ScanLine_Tiling=ScalarParameter() {defaultValue=30, parameterName=Scanline Tiling 2}
    Scanline_PannerSpeed=ScalarParameter() {defaultValue=1, parameterName=Scanline Speed 2}
    Scanline_Rotation
    UVs
   ) {materialFunction=/Game/Global/Materials/MaterialGym/Materials/HolographicTexture/MF_HolographicTexture_ScanLines}.Result) {bias=-0.5}
  n13 = MaterialFunctionCall!(
   Blend Range=ScalarParameter() {defaultValue=200, parameterName=Scanline Fade Size}
   Start Offset=ScalarParameter() {defaultValue=-300, parameterName=Scanline Fade Distance}
  ) {materialFunction=/Engine/Functions/Engine_MaterialFunctions03/Particles/Distance_Blend}
  n14 = Add(A=Multiply(A=@n12, B=OneMinus(None=@n13.Result)))
  Multiply(A=@n9, B=@n14)
  n15 = MaterialFunctionCall!(
   Time
   Glitch Flickering=ScalarParameter() {defaultValue=0.7, parameterName="Glitch Flickering (S)"}
   GlitchDuration=ScalarParameter() {defaultValue=0.075, group=HolographicTexture_Glitch, parameterName=GlitchDuration, sortPriority=10}
   GlitchFrequency=ScalarParameter() {defaultValue=0.05, group=HolographicTexture_Glitch, parameterName=GlitchFrequency, sortPriority=10}
   GlitchMaskResolution
   PixellateGlitch_MinResolution
   PixellateGlitch_MaxResolution
   UV
  ) {materialFunction=/Hermes/Techart/UtopiaEmissiveSurfaces/Glass_4Layer/Glitching/MF_MMGlitchingScreens}
  n16 = PerInstanceRandom()
  NamedRerouteDeclaration(None=Add(A=@n15.Random time, B=LinearInterpolate(Alpha=@n16) {constB=1234})) {name=Random Time, nodeColor=<rgba 0 1 0.6484 1>}
  NamedRerouteDeclaration(None=LinearInterpolate(Alpha=@n15.GlitchFrequencyMask) {constA=0.3, constB=0.9}) {name=GlitchFrequencyMask, nodeColor=<rgba 0.9377 1 0.8953 1>}
  n17 = NamedRerouteUsage()
  n18 = MaterialFunctionCall(Float2=@n17) {materialFunction=&BreakOutFloat2Components}
  n19 = Add(A=NamedRerouteUsage(), B=AppendVector(A=Add(A=@n18.R) {constB=0.007}, B=@n18.G))
  n20 = Multiply(A=Add(A=TextureCoordinate()) {constB=-0.5}, B=ScalarParameter() {defaultValue=3.5668}) {constB=4}
  n21 = Saturate(None=Add(A=Abs(None=@n20)) {constB=-1})
  n22 = MaterialFunctionCall(Float2=Multiply(A=@n21, B=@n21)) {materialFunction=&BreakOutFloat2Components}
  n23 = SquareRoot(None=Add(A=@n22.R, B=@n22.R))
  n24 = Saturate(None=Multiply(A=@n23) {constB=2})
  n25 = OneMinus(None=NamedRerouteUsage())
  NamedRerouteDeclaration(None=LinearInterpolate(A=LinearInterpolate(A=@n19, B=@n17, Alpha=@n24), B=@n17, Alpha=@n25)) {name=UV Glitching, nodeColor=<rgba 1 0 0.3281 1>}
  n26 = NamedRerouteDeclaration(None=@n15.Square Mask) {name=Square Color glitch, nodeColor=<rgba 0.2422 1 0 1>}
  NamedRerouteDeclaration(None=Multiply(A=Add(A=@n26) {constB=-0.5}) {constB=0.05}) {name=Glitch Noise, nodeColor=<rgba 0 1 0.6953 1>}
  NamedRerouteUsage()
  OneMinus(None=@n15.GlitchFrequencyMask)
  Round(None=@n23)
  Saturate(None=Add(A=@n15.GlitchFrequencyMask) {constB=0.3})
  Saturate(None=Multiply(A=@n25) {constB=2})
  n27 = NamedRerouteUsage()
  n28 = Add(A=Time(), B=@n3)
  n29 = Multiply(A=@n28) {constB=4}
  n30 = Multiply(A=ConstantBiasScale(None=Sine(None=@n29) {period=1.9967}), B=ConstantBiasScale(None=Sine(None=@n29) {period=4.2494}))
  n31 = ConstantBiasScale(None=Sine(None=@n29) {period=6.0273})
  n32 = ComponentMask(None=TextureCoordinate()) {g=true}
  n33 = ConstantBiasScale(None=Sine(None=Add(A=@n32, B=@n28)) {period=2})
  n34 = ScalarParameter() {defaultValue=0.3, parameterName=Scanline Distortion Big}
  n35 = Multiply(
   A=Multiply(A=ScalarParameter() {defaultValue=0.008, parameterName=Scanline Distortion}, B=Multiply(A=@n30, B=@n31))
   B=Add(A=@n12, B=Multiply(A=@n33, B=@n34))
  )
  n36 = TextureSample(UVs=StaticSwitchParameter(True=Add(A=@n27, B=@n35), False=@n7) {group=Projector, parameterName=Mask Maker Take Over, sortPriority=1}) {samplerSource=SSM_Wrap_WorldGroupSettings, texture=&T_Labs_BookAtlas}
  n37 = Fresnel() {baseReflectFraction=0, exponent=2}
  n38 = Multiply(A=Multiply(A=Multiply(A=@n36.RGB, B=@n14), B=OneMinus(None=@n37)), B=ScalarParameter() {defaultValue=0.8, parameterName=Intensity})
  n39 = Multiply(A=Power(Base=Desaturation(None=@n38)) {constExponent=3}) {constB=5}
  n40 = LinearInterpolate(
   A=@n38
   B=LinearInterpolate(
    A=VectorParameter() {parameterName=Color 1, defaultValue=<rgba 0.0833 0.0006 0.0369 1>}.RGB
    B=VectorParameter() {parameterName=Color 2, defaultValue=<rgba 0.2812 0.019 0.1515 1>}.RGB
    Alpha=Saturate(None=@n39)
   )
   Alpha=NamedRerouteUsage()
  )
  n41 = Constant() {r=0.1}
  n42 = Constant() {r=0.5}
  n43 = TextureCoordinate()
  n44 = Abs(None=ConstantBiasScale(None=ComponentMask(None=@n43) {g=true, r=true}) {bias=-0.5, scale=2})
  n45 = ScalarParameter() {defaultValue=32, parameterName=EdgeMaskPower}
  n46 = MaterialFunctionCall(Float2=OneMinus(None=Power(Base=@n44, Exp=@n45) {constExponent=32})) {materialFunction=&BreakOutFloat2Components}
  n47 = Multiply(A=Multiply(A=@n46.R, B=@n46.R), B=Power(Base=@n9) {constExponent=0.5})
  n48 = EyeAdaptationInverse(
   LightValueInput=StaticSwitchParameter(True=@n40, False=@n38) {group=Projector, parameterName=Mask Maker Take Over, sortPriority=1}
   AlphaInput=ScalarParameter() {defaultValue=0.75, parameterName=ExposureCompensation}
  )
  n49 = Multiply(A=@n38, B=ScalarParameter() {defaultValue=1, parameterName=LumenContribution})
  n50 = Power(Base=ComponentMask(None=TextureCoordinate()) {g=true}, Exp=ScalarParameter() {defaultValue=6, parameterName=LitGradientPower}) {constExponent=8}
  n51 = StaticSwitchParameter(
   True=VectorParameter() {parameterName=LitGradientColor MM, defaultValue=<rgba 0.0312 0 0.0009 1>}.RGB
   False=VectorParameter() {parameterName=LitGradientColor, defaultValue=<rgba 0 0.0933 0.7 1>}.RGB
  ) {group=Projector, parameterName=Mask Maker Take Over, sortPriority=1}
  n52 = SetMaterialAttributes!(
   MaterialAttributes
   Base Color=StaticSwitchParameter(True=@n40, False=@n38) {group=Projector, parameterName=Mask Maker Take Over, sortPriority=1}
   Metallic=Constant() {r=1}
   Specular=Constant() {r=0.2}
   Roughness=MaterialFunctionCall("Xb1,PS4"=@n41, "PC-High&&Epic,XSX,PS5"=@n41, "PC-Perf,Android"=@n42, Switch=@n42, "PC-Low(NotPerf)&&Medium"=@n41) {materialFunction=&PlatformSwitch_1D}.Result
   Anisotropy
   Emissive Color=Add(A=RayTracingQualitySwitch(Normal=@n48.EyeAdaptationInverse, RayTraced=@n49), B=Multiply(A=@n50, B=@n51))
   Opacity
   Opacity Mask=MaterialFunctionCall(
    "Xb1,PS4"=@n9
    "PC-High&&Epic,XSX,PS5"=MaterialFunctionCall(Alpha Threshold=@n47) {materialFunction=&DitherTemporalAA}.Result
    "PC-Perf,Android"=@n9
    Switch=@n9
    "PC-Low(NotPerf)&&Medium"=@n9
   ) {materialFunction=&PlatformSwitch_1D}.Result
  ) {attributeSetTypes [
    #1
    #2
    #3
    #4
    #5
    #6
    #7
    #8
   ]}
  n53 = MaterialFunctionCall(Material_In=@n52) {materialFunction=&MF_FN_DestructibleDamage}
  SetMaterialAttributes!(MaterialAttributes=@n53.Material_Out, World Position Offset=@n53.Material_Out) {attributeSetTypes [#9]}
 ]
}
